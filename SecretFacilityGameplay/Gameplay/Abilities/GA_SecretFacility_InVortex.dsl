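; GA_SecretFacility_InVortex is a Secret Facility gameplay ability that puts the avatar player pawn into a vortex state while it is active.
; Activation calls SetinVortex on the Fort player pawn with the vortex enabled and hard-coded tuning values.
; Ending the ability calls it again with the vortex disabled and values zeroed, regardless of cancellation.
; Both transitions print a debug string, and the activate-from-event handler is empty.
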(event Ability|EventActivateAbilityFromEvent
  (params (EventData :type Gameplay%20Event%20Data%20Structure%20%28by%20ref%29)))

(event Ability|EventActivateAbility
  (Development|PrintString "Secret Facility: New Vortex Start")
  (Vortex|SetinVortex (CallFunction|GetAvatarasFortPlayerPawn) true 2e+03 -8e+03 0.0 0.0 0.0 true))

(event Ability|EventOnEndAbility
  (params (bWasCancelled :type Boolean))
  (Development|PrintString "Secret Facility: New Vortex End")
  (Vortex|SetinVortex (CallFunction|GetAvatarasFortPlayerPawn) false 0.0 0.0 0.0 0.0 0.0 false))
