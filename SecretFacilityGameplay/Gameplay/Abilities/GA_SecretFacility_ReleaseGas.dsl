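; Gameplay ability that releases invisibility gas from the Secret Facility invisibility button actor.
; On activation it applies the gas-active effects to the owning actor, then casts that actor to BGA_SecretFacility_InvisibilityButton.
; It applies the invisibility gas effects to each FortPawn overlapping the button's gas collision box at that moment, using a one-time overlap query.
; The end-ability event is empty, and the graph never ends the ability itself.
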(event Ability|EventActivateAbility
  (bind _returnvalue (Ability|GetOwningActorfromActorInfo))
  (Fortnite|ApplyGameplayEffectsToActor _returnvalue (Variables|Default|GetGasActiveEffects) (Transformation|GetActorLocation _returnvalue))
  (bind _asbga_secret_facility_invisibility_button (Utilities|Casting|CastToBGA_SecretFacility_InvisibilityButton _returnvalue))
  (for _array_element (Collision|GetOverlappingActors (Variables|/SecretFacilityGameplay/Gameplay/Actors/BGA_SecretFacility_InvisibilityButton.BGA_SecretFacility_InvisibilityButton_C|GetGasCollisionBox _asbga_secret_facility_invisibility_button) "/Script/FortniteGame.FortPawn" :!class /Script/Engine.PrimitiveComponent)
    (Fortnite|ApplyGameplayEffectsToActor _array_element (Variables|Default|GetInvisibilityGasEffects) (Transformation|GetActorLocation _array_element))))

(event Ability|EventOnEndAbility
  (params (bWasCancelled :type Boolean)))
